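# Physics asset for the Pickaxe_Sleek_Male cosmetic melee pickaxe mesh, giving the weapon a simple collision shape.
# A single body on the Melee_TwoHanded_Handle bone uses one capsule, slightly offset from the bone and rotated relative to it.
# With only one body, it has no constraints, so the pickaxe simulates as one rigid piece rather than an articulated ragdoll.
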PhysicsAsset /BRCosmetics/Weapons/FORT_Melee/Pickaxe_Sleek_Male/Meshes/Pickaxe_Sleek_Male_PhysicsAsset
bodies {Melee_TwoHanded_Handle [{shapeType=Capsule, length=59.5407, radius=19.0181, center=<0.9282 -2.0371 21.1419>, rotation=<rot 1.6992 -90 -0.0311>}]}
constraints []
